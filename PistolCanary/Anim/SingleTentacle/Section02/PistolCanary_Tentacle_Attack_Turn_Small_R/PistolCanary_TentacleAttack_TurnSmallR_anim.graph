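# Level Sequence for the PistolCanary single-tentacle small right-turn attack animation, judging by its name and path.
# One Tentacle Pistol Canary Creature skeletal mesh actor binding carries a spawn track, a fixed world transform and a Tentacle_PistolCanary_Creature_CtrlRig track.
# The rig track has two sections, the second zero-scaled like an additive layer.
# Both expose only the "control rig options" transform and a weight, with no varying motion shown.
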LevelSequence /PistolCanary/Anim/SingleTentacle/Section02/PistolCanary_Tentacle_Attack_Turn_Small_R/PistolCanary_TentacleAttack_TurnSmallR_anim
bindings [Tentacle Pistol Canary Creature {
  id=#1
  possessed_class=SkeletalMeshActor
  tracks [
   {type=MovieSceneSpawnTrack, sections [{type=MovieSceneSpawnSection, channels [None {type=MovieSceneScriptingBoolChannel, default=true}]}]}
   {
    type=MovieScene3DTransformTrack
    sections [{
      type=MovieScene3DTransformSection
      channels [
       Location.X {type=MovieSceneScriptingDoubleChannel, default=14287.208}
       Location.Y {type=MovieSceneScriptingDoubleChannel, default=70618.9219}
       Location.Z {type=MovieSceneScriptingDoubleChannel, default=3947.8015}
       Rotation.X {type=MovieSceneScriptingDoubleChannel, default=0}
       Rotation.Y {type=MovieSceneScriptingDoubleChannel, default=0}
       Rotation.Z {type=MovieSceneScriptingDoubleChannel, default=0}
       Scale.X {type=MovieSceneScriptingDoubleChannel, default=1}
       Scale.Y {type=MovieSceneScriptingDoubleChannel, default=1}
       Scale.Z {type=MovieSceneScriptingDoubleChannel, default=1}
      ]
     }]
   }
   {
    type=MovieSceneControlRigParameterTrack
    display_name=Tentacle_PistolCanary_Creature_CtrlRig
    sections [
     {
      type=MovieSceneControlRigParameterSection
      channels [
       control rig options.Location.X {type=MovieSceneScriptingFloatChannel, default=0, const [0,2]}
       control rig options.Location.Y {type=MovieSceneScriptingFloatChannel, default=0, const [0,2]}
       control rig options.Location.Z {type=MovieSceneScriptingFloatChannel, default=0, const [0,2]}
       control rig options.Rotation.X {type=MovieSceneScriptingFloatChannel, default=0, const [0,2]}
       control rig options.Rotation.Y {type=MovieSceneScriptingFloatChannel, default=0, const [0,2]}
       control rig options.Rotation.Z {type=MovieSceneScriptingFloatChannel, default=0, const [0,2]}
       control rig options.Scale.X {type=MovieSceneScriptingFloatChannel, default=1, const [1,2]}
       control rig options.Scale.Y {type=MovieSceneScriptingFloatChannel, default=1, const [1,2]}
       control rig options.Scale.Z {type=MovieSceneScriptingFloatChannel, default=1, const [1,2]}
       Weight {type=MovieSceneScriptingFloatChannel, default=1}
      ]
     }
     {
      type=MovieSceneControlRigParameterSection
      channels [
       control rig options.Location.X {type=MovieSceneScriptingFloatChannel, default=0, const [0,1]}
       control rig options.Location.Y {type=MovieSceneScriptingFloatChannel, default=0, const [0,1]}
       control rig options.Location.Z {type=MovieSceneScriptingFloatChannel, default=0, const [0,1]}
       control rig options.Rotation.X {type=MovieSceneScriptingFloatChannel, default=0, const [0,1]}
       control rig options.Rotation.Y {type=MovieSceneScriptingFloatChannel, default=0, const [0,1]}
       control rig options.Rotation.Z {type=MovieSceneScriptingFloatChannel, default=0, const [0,1]}
       control rig options.Scale.X {type=MovieSceneScriptingFloatChannel, default=0, const [0,1]}
       control rig options.Scale.Y {type=MovieSceneScriptingFloatChannel, default=0, const [0,1]}
       control rig options.Scale.Z {type=MovieSceneScriptingFloatChannel, default=0, const [0,1]}
       Weight {type=MovieSceneScriptingFloatChannel, default=1, const [1,1]}
      ]
     }
    ]
   }
  ]
 }]
display_rate {denominator=1, numerator=30}
tick_resolution {denominator=1, numerator=24000}
tracks []
